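; ToyFox footsteps targeting filter that removes building actors lying beyond a set distance from the AOE source.
; It casts the hit actor to BuildingActor and compares the squared distance from its centroid to the hit result's trace start against the square of the scalable-float Distance.
; Results without a valid hit actor are kept.
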(fn ShouldRemoveTarget
  (params (TargetingHandle :type Targeting%20Request%20Handle%20Structure%20%28by%20ref%29) (TargetData :type Targeting%20Default%20Result%20Data%20Structure%20%28by%20ref%29))
  (bind _hitresult (Utilities|Struct|BreakTargetingDefaultResultData TargetData))
  (bind (_bblockinghit _binitialoverlap _time _distance _location _impactpoint _normal _impactnormal _physmat _hitactor _hitcomponent _hitbonename _bonename _hititem _elementindex _faceindex _tracestart) (Collision|BreakHitResult _hitresult))
  ;[0 0.0827 0.15 0.5] Remove targets whose actor location are not within the specified distance from the AOE source location.
  (Utilities|FlowControl|Sequence
    (:then_0
      (Utilities|IsValid _hitactor
        (:Is%20Valid
          (bind _asbuilding_actor (Utilities|Casting|CastToBuildingActor _hitactor))
          (return (> (Math|Vector|DistanceSquared(Vector) (Building|GetCentroid _asbuilding_actor) _tracestart) (Math|Float|Square (ScalableFloat|GetValueatLevel (Variables|Default|GetDistance))))))))
    (:then_1
      (return false))))
